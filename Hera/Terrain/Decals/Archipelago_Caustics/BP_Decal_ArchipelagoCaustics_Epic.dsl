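; Decal Blueprint for archipelago water caustics, likely an Epic-quality variant, that passes per-instance tuning values to its decal material.
; The construction script packs texture scale, desaturation, soft radius and sides mask variables into the RGBA channels of a color and applies it through SetDecalColor.
; The decal color thus serves as a four-value parameter channel rather than a tint.
; BeginPlay and Tick are empty.
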(event EventBeginPlay)

(event EventTick
  (params (DeltaSeconds :type Float%20%28single-precision%29)))

(fn ConstructionScript
  (Rendering|Components|Decal|SetDecalColor (Variables|Decal|GetDecal) (Math|Color|MakeColor (Variables|Default|GetTexture%20Scale%20Bias) (Variables|Default|GetDesaturation%20Bias) (Variables|Default|GetSoft%20Radius%20Bias) (Variables|Default|GetSides%20Mask%20Multiply))))
