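# Material function that produces animated water caustics color from world-position UVs and time, likely for Archipelago terrain decals.
# Water velocity offsets the UVs in two half-cycle-shifted flow phases, both exposed as outputs, though only phase A feeds the texture lookup.
# A 6x6 flipbook blends consecutive frames, using derivatives from pre-Frac UVs for mip selection.
# Brightened caustics lerp from a fadeout brightness according to water depth over a gradient length.
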MaterialFunction /Hera/Terrain/Decals/Archipelago_Caustics/MF_WaterCaustics_Custom
graph {nodes [
  Constant2Vector() {g=1, r=1}
  n1 = NamedRerouteUsage()
  n2 = MaterialFunctionCall(Speed=ScalarParameter() {defaultValue=0.5, group=Caustics, parameterName=Caustics Phase Fade Time}) {materialFunction=&TimeWithSpeedVariable}
  n3 = Subtract(A=@n2.Time) {constB=0.5, desc=phase offset}
  n4 = Divide(
   A=Divide(
    A=FunctionInput() {bUsePreviewValueAsDefault=true, inputName=Water Velocity, inputType=FunctionInput_Vector2, sortPriority=5}
    B=NamedRerouteUsage()
   )
   B=ScalarParameter() {defaultValue=0.5, group=Caustics, parameterName=Caustics Phase Fade Time}
  )
  n5 = Subtract(A=@n1, B=Multiply(A=Frac(None=@n3), B=@n4))
  n6 = VectorParameter() {parameterName=Flowmap UV2 Offset and Scale, defaultValue=<rgba 0 0 0.85 0>}
  n7 = ComponentMask(None=@n6.RGB) {b=true}
  n8 = FunctionOutput(None=Add(A=Divide(A=@n5, B=@n7), B=ComponentMask(None=@n6.RGB) {g=true, r=true}) {constB=0.5}) {
   description="R= amount of distortion for lerp sample A. G= distortion amount for Lerp input B.  B= Lerp Alpha. A= Lerped Distortion amounts. This is useful for modulating normals by the amount of movement that occurs in the texture. I.e. Lerp ((0,0,1), (This nodes normal output), (This nodes distortions alpha channel)) to only display the normals where the waves are currently highest."
   outputName=UV B
   sortPriority=6
  }
  n9 = Constant() {r=6}
  Divide(A=Frac(None=@n8), B=@n9)
  n10 = NamedRerouteUsage()
  n11 = TextureObject() {samplerType=SAMPLERTYPE_LinearColor, texture=/FortWater/Textures/T_Siren_Caustics_03}
  n12 = Divide(A=NamedRerouteUsage()) {constB=6}
  n13 = DDX(Value=@n12)
  n14 = DDY(Value=@n12)
  n15 = NamedRerouteUsage()
  n16 = Constant() {r=6}
  n17 = FunctionInput(Preview=MaterialFunctionCall(Speed=ScalarParameter() {defaultValue=0.5, group=Caustics, parameterName=WaterCausticsTimeSpeed}) {materialFunction=&TimeWithSpeedVariable}.Time) {bUsePreviewValueAsDefault=true, inputName=Time, inputType=FunctionInput_Scalar, sortPriority=3, previewValue=<1024 0 0 1>}
  n18 = Multiply(A=Multiply(A=@n16, B=@n16), B=Frac(None=Add(A=@n17) {constB=0.5}))
  n19 = Multiply(
   A=LinearInterpolate(
    A=TextureSample!(UVs=@n10, Tex=@n11, "DDX(UVs)"=@n13, "DDY(UVs)"=@n14, Apply View MipBias) {mipValueMode=TMVM_Derivative}.RGB
    B=TextureSample(UVs=@n15, Tex=@n11, "DDX(UVs)"=@n13, "DDY(UVs)"=@n14) {mipValueMode=TMVM_Derivative}.RGB
    Alpha=Frac(None=@n18)
   )
   B=ScalarParameter() {defaultValue=1, group=Caustics, parameterName=Caustics Brightness}
  )
  n20 = ScalarParameter() {defaultValue=0.2, group=Caustics, parameterName=Caustics Add}
  n21 = NamedRerouteUsage()
  n22 = FunctionInput(Preview=Subtract(A=ComponentMask(None=@n21) {b=true}, B=ComponentMask(None=@n21) {b=true})) {bUsePreviewValueAsDefault=true, inputName=Water Depth, inputType=FunctionInput_Scalar, sortPriority=2, previewValue=<1024 0 0 1>}
  n23 = Divide(A=@n22, B=ScalarParameter() {defaultValue=256, group=Caustics, parameterName=Caustics Gradient Length})
  FunctionOutput(None=LinearInterpolate(
    A=ScalarParameter() {defaultValue=1, group=Caustics, parameterName=Caustics Fadeout Brightness}
    B=Add(A=@n19, B=@n20)
    Alpha=Saturate(None=@n23)
   ) {constA=0.75}) {outputName=ColoredResult}
  n24 = NamedRerouteUsage()
  Multiply(A=Divide(A=TextureCoordinate(), B=Multiply(A=@n24) {constB=0.5}), B=@n24)
  NamedRerouteDeclaration(None=FunctionInput(Preview=WorldPosition().XYZ) {bUsePreviewValueAsDefault=true, inputName=WorldPosition}) {name=Wordl Pos, nodeColor=<rgba 1 0 0.7266 1>}
  n25 = FunctionInput() {bUsePreviewValueAsDefault=true, inputName=Scale, inputType=FunctionInput_Scalar, sortPriority=1, previewValue=<1024 0 0 1>}
  NamedRerouteDeclaration(None=@n25) {name=TEx Scale, nodeColor=<rgba 0.0078 1 0 1>}
  n26 = Frac(None=@n2.Time)
  n27 = FunctionOutput(None=Subtract(A=@n1, B=Multiply(A=@n4, B=@n26))) {
   description="R= amount of distortion for lerp sample A. G= distortion amount for Lerp input B.  B= Lerp Alpha. A= Lerped Distortion amounts. This is useful for modulating normals by the amount of movement that occurs in the texture. I.e. Lerp ((0,0,1), (This nodes normal output), (This nodes distortions alpha channel)) to only display the normals where the waves are currently highest."
   outputName=UV A
   sortPriority=5
  }
  n28 = Divide(A=Frac(None=@n27), B=@n9)
  n29 = Constant() {r=6}
  n30 = Add(A=@n17) {constB=0.5}
  n31 = MaterialFunctionCall!(1D Index=Multiply(A=Multiply(A=@n29, B=@n29), B=Frac(None=@n30)), Cells XY=@n29) {materialFunction=/Engine/Functions/Engine_MaterialFunctions01/Coordinates/1Dto2DIndex}
  NamedRerouteDeclaration(None=Add(A=@n28, B=Divide(A=@n31.2D Index, B=@n29))) {name=UV A, nodeColor=<rgba 1 0 0.3281 1>}
  n32 = MaterialFunctionCall(1D Index=Add(A=@n18), Cells XY=@n16) {materialFunction=&1Dto2DIndex}
  NamedRerouteDeclaration(None=Add(A=@n28, B=Divide(A=@n32.2D Index, B=@n16))) {name=UV B, nodeColor=<rgba 1 0 0.3281 1>}
  n33 = NamedRerouteUsage()
  NamedRerouteDeclaration(None=Divide(A=ComponentMask(None=@n33) {g=true, r=true}, B=@n24)) {name=World UVs, nodeColor=<rgba 0.9688 1 0 1>}
  ScalarParameter() {defaultValue=10, parameterName=Caustics Speed}
 ]}
