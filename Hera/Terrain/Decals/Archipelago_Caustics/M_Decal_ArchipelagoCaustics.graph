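# Translucent, unlit deferred decal material that draws animated water caustics for the Archipelago terrain.
# A custom caustics function, scaled partly by object Z position, produces the pattern, which is desaturated and tinted by time-of-day scalar and color functions.
# A soft UV sphere mask and scene world normals shape a second mask favoring downward and side-facing surfaces.
# Decal color channels adjust each instance's scale, desaturation, mask radius and side-surface strength.
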Material /Hera/Terrain/Decals/Archipelago_Caustics/M_Decal_ArchipelagoCaustics
graph {
 properties {
  blendMode=BLEND_Translucent
  materialDomain=MD_DeferredDecal
  pixelDepthOffsetMode=PDOM_Legacy
  shadingModel=MSM_Unlit
  usedShadingModels=Unlit
 }
 nodes [
  MaterialFunctionCall(SceneDepth=SceneDepth().R) {materialFunction=&WorldPositionBehindTranslucency}
  Multiply()
  n1 = ScalarParameter() {defaultValue=4000, parameterName="Scale (S)"}
  n2 = DecalColor()
  n3 = ComponentMask(None=ObjectPositionWS()) {b=true}
  n4 = MaterialFunctionCall!(
   WorldPosition
   Scale=Add(A=Add(A=@n1, B=@n2.R), B=Multiply(A=@n3))
   Water Depth=ScalarParameter() {defaultValue=300, parameterName="Water Depth (S)"}
   Time
   Water Velocity
  ) {materialFunction=&/MF_WaterCaustics_Custom}
  n5 = ComponentMask(None=@n4.ColoredResult) {r=true}
  n6 = ScalarParameter() {defaultValue=1.25, parameterName=Exp_2}
  n7 = Multiply(A=@n4.ColoredResult, B=Clamp(None=Power(Base=@n5, Exp=@n6)))
  n8 = Desaturation(None=@n7, Fraction=Add(A=ScalarParameter() {defaultValue=-0.5, parameterName=Caustic Desaturation}, B=DecalColor().G))
  n9 = ScalarParameter() {defaultValue=0.4, parameterName=TOD_MorningEvening}
  n10 = MaterialFunctionCall(
   Value_Morning=@n9
   Value_Day=ScalarParameter() {defaultValue=0.4, parameterName=TOD_Day}
   Value_Evening=@n9
   Value_Night=ScalarParameter() {defaultValue=0.3, parameterName=TOD_Night}
  ) {materialFunction=&Scalar_Time_of_Day}
  n11 = MaterialFunctionCall() {materialFunction=&Color_Time_of_Day}
  n12 = ScalarParameter() {defaultValue=0.7, parameterName=TOD Desaturation}
  Reroute(None=Multiply(A=Multiply(A=@n8, B=@n10.Result), B=Desaturation(None=@n11.Result, Fraction=@n12)))
  n13 = Multiply(
   A=VectorParameter() {parameterName=Normal Direction, defaultValue=<rgba 0.3 0.3 1 1>}.RGB
   B=ComponentMask(None=SceneTexture() {sceneTextureId=PPI_WorldNormal}.Color) {b=true, g=true, r=true}
  )
  n14 = Saturate(None=Multiply(A=ComponentMask(None=@n13) {b=true}) {constB=-1})
  n15 = Abs(None=Multiply(A=@n13, B=DecalColor().A))
  n16 = Saturate(None=Add(A=ComponentMask(None=@n15) {r=true}, B=ComponentMask(None=@n15) {g=true}))
  n17 = ScalarParameter() {defaultValue=0.5, parameterName=Sphere_Radius}
  n18 = DecalColor()
  n19 = Multiply(
   A=@n7
   B=SphereMask(
    A=TextureCoordinate()
    B=Constant2Vector() {g=0.5, r=0.5}
    Radius=Add(A=@n17, B=@n18.B)
    Hardness=ScalarParameter() {parameterName=Sphere_Hardness}
   )
  )
  n20 = ScalarParameter() {defaultValue=1, parameterName=Sphere_Opacity}
  Reroute(None=Multiply(A=Add(A=@n14, B=@n16), B=Multiply(A=@n19, B=@n20)))
 ]
}
